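# Translucent, unlit, two-sided Niagara master material for airflow blur effects on sprites, ribbons and meshes.
# Opacity multiplies panning caustic-noise and smoke-trail textures with UV edge masks, particle alpha, depth fade and view-angle fade.
# Dynamic parameters drive pan speed, a noise dissolve along U and a UV offset (named WPO Offset, added to vertex red).
# Emissive is boosted particle color, desaturated by a second dynamic parameter and passed through MF_FX_EmissiveAdjustments.
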Material /Game/Effects/Niagara/MasterMaterials/Airflow/M_AirflowBlur
graph {
 properties {
  bUsedWithNiagaraMeshParticles=true
  bUsedWithNiagaraRibbons=true
  bUsedWithNiagaraSprites=true
  blendMode=BLEND_Translucent
  pixelDepthOffsetMode=PDOM_Legacy
  shadingModel=MSM_Unlit
  twoSided=true
  usedShadingModels=Unlit
 }
 nodes [
  n1 = ParticleColor()
  n2 = ParticleColor()
  n3 = Sine(None=ComponentMask(None=TextureCoordinate()) {g=true}) {period=2}
  n4 = VectorParameter() {parameterName=Tex 3 Tex 4 Tile, channelNames {a=4 Y, b=4 X, g=3 Y, r=3 X}, defaultValue=<rgba 1 1 1 1>}
  n5 = Add(A=Multiply(A=TextureCoordinate() {uTiling=0.5, vTiling=0.5}, B=ComponentMask(None=@n4.RGB) {g=true, r=true})) {constB=0}
  n6 = LinearInterpolate(
   A=Multiply(A=Time(), B=ScalarParameter() {defaultValue=1, parameterName=Panner speed})
   B=DynamicParameter() {defaultValue=<rgba 1 1 70 1>, paramNames [WPO Offset, WPO, Speed, Dis]}.Speed
   Alpha=ScalarParameter() {defaultValue=1, parameterName=Use Custom Speed}
  )
  n7 = Power(Base=TextureSample(UVs=Panner(Coordinate=@n5, Time=@n6) {speedX=2, speedY=0.05}) {samplerType=SAMPLERTYPE_LinearGrayscale, texture=/Game/Global/Textures/Default/Caustics/T_Nosie_Caustics_Grayscale}.R)
  n8 = SphereMask(A=TextureCoordinate() {uTiling=0}, B=Constant2Vector() {g=0.5}) {attenuationRadius=1, hardnessPercent=0}
  n9 = Add(A=Power(Base=OneMinus(None=@n8)))
  n10 = Multiply(A=Saturate(None=Multiply(A=@n3) {constB=4}), B=Add(A=Multiply(A=@n7) {constB=0.1}, B=Multiply(A=@n9) {constB=0.03}))
  n11 = Multiply(A=ComponentMask(None=TextureCoordinate()) {r=true}, B=ScalarParameter() {defaultValue=20, parameterName=Lead Edge mask}) {constB=20}
  n12 = TextureCoordinate()
  n13 = Power(Base=OneMinus(None=ComponentMask(None=@n12) {r=true})) {constExponent=1.5}
  n14 = Multiply(A=Multiply(A=Multiply(A=@n2.A, B=@n10), B=Saturate(None=@n11)), B=Saturate(None=Multiply(A=@n13) {constB=5}))
  n15 = TextureCoordinate() {vTiling=0.3}
  n16 = VectorParameter() {parameterName=Tex1 Tex 2 Tile, channelNames {a=2 Y, b=2 X, g=1 Y, r=1 X}, defaultValue=<rgba 1 1 1 1>}
  n17 = ComponentMask(None=@n16.RGB) {g=true, r=true}
  n18 = Add(A=DynamicParameter() {defaultValue=<rgba 1 1 70 1>, paramNames [WPO Offset, WPO, Speed, Dis]}.WPO Offset, B=VertexColor().R)
  n19 = Constant()
  n20 = Time()
  n21 = ScalarParameter() {defaultValue=1, parameterName=Panner speed}
  n22 = Panner(
   Coordinate=Add(A=Multiply(A=@n15, B=@n17), B=AppendVector(A=@n18, B=@n19))
   Time=LinearInterpolate(
    A=Multiply(A=@n20, B=@n21)
    B=DynamicParameter() {defaultValue=<rgba 1 1 70 1>, paramNames [WPO Offset, WPO, Speed, Dis]}.Speed
    Alpha=ScalarParameter() {defaultValue=1, parameterName=Use Custom Speed}
   )
  ) {speedX=2}
  n23 = TextureCoordinate()
  n24 = LinearInterpolate(
   A=ScalarParameter() {parameterName=Dissolve mask Min}
   B=DynamicParameter() {defaultValue=<rgba 1 1 70 1>, paramNames [WPO Offset, WPO, Speed, Dis]}.Dis
   Alpha=Power(Base=ComponentMask(None=@n23) {r=true}) {constExponent=1.3}
  ) {constA=0.1, constB=2}
  n25 = Saturate(None=Power(
    Base=TextureSample(UVs=@n22) {samplerType=SAMPLERTYPE_LinearGrayscale, texture=&T_Nosie_Caustics_Grayscale}.R
    Exp=Saturate(None=@n24)
   ))
  n26 = TextureCoordinate()
  n27 = ComponentMask(None=@n16.RGBA) {a=true, b=true}
  n28 = Multiply(A=Time(), B=ScalarParameter() {defaultValue=1, parameterName=Panner speed})
  n29 = DynamicParameter() {defaultValue=<rgba 1 1 70 1>, paramNames [WPO Offset, WPO, Speed, Dis]}
  n30 = ScalarParameter() {defaultValue=1, parameterName=Use Custom Speed}
  n31 = TextureSample(UVs=Panner(Coordinate=Multiply(A=@n26, B=@n27), Time=LinearInterpolate(A=@n28, B=@n29.Speed, Alpha=@n30)) {speedX=2}) {samplerType=SAMPLERTYPE_Grayscale, texture=/Game/Effects/Fort_Effects/Textures/Smoke/T_Smoke_Trail_Soft_NoClamp}
  n32 = DepthFade(Opacity=Multiply(A=Multiply(A=@n14, B=@n25), B=Power(Base=@n31.R)) {constA=1}) {fadeDistanceDefault=30}
  n33 = ScalarParameter() {defaultValue=3, parameterName="FadeSensitivity (optional) (S)"}
  DepthFade(
   Opacity=Multiply(
    A=Multiply(A=@n1.A, B=@n32)
    B=MaterialFunctionCall("FadeSensitivity (optional)"=@n33) {materialFunction=&FadeBasedOnViewAngle}.Result
   )
   FadeDistance=ScalarParameter() {defaultValue=50, parameterName=FadeDistance}
  ) {fadeDistanceDefault=50}
  n34 = Multiply(A=@n2.RGB) {constB=10}
  MaterialFunctionCall(Emissive=Desaturation(
    None=Multiply(A=@n34, B=@n1.RGB)
    Fraction=DynamicParameter() {parameterIndex=1, defaultValue=<rgba 0 1 1 1>, paramNames [Desaturation,Param2,Param3,Param4]}.Desaturation
   )) {materialFunction=&MF_FX_EmissiveAdjustments}
 ]
}
