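# Translucent, unlit, two-sided master material for Niagara mesh particles that draws sharp airflow streaks, with defaults and options from the Energize and Stalemate emote effects.
# Opacity combines a masked panning texture, particle alpha, a dynamic-parameter-driven height-lerp dissolve, view-angle fading and depth fade, with optional gradient edge masking.
# Emissive tints a UV gradient or noise mask by particle color.
# Optional SplineThicken world position offset adjusts width, optionally adding per-particle-randomized noise displacement.
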Material /Game/Effects/Niagara/MasterMaterials/Airflow/M_AirflowSharp
graph {
 properties {
  bUsedWithNiagaraMeshParticles=true
  blendMode=BLEND_Translucent
  pixelDepthOffsetMode=PDOM_Legacy
  shadingModel=MSM_Unlit
  twoSided=true
  usedShadingModels=Unlit
 }
 nodes [
  n1 = LinearInterpolate(Alpha=VertexColor().R) {constA=0.25}
  n2 = ScalarParameter() {parameterName=MultSplineThickenWithVertColor}
  n3 = Multiply(
   A=MaterialFunctionCall(
    WidthBase=ScalarParameter() {defaultValue=1, parameterName=SplineThicken_WidthBase}
    WidthTip=ScalarParameter() {defaultValue=1, parameterName=SplineThicken_WidthTip}
   ) {materialFunction=&SplineThicken}.WorldPositionOffset
   B=Multiply(
    A=LinearInterpolate(B=@n1, Alpha=@n2) {constA=1}
    B=DynamicParameter() {defaultValue=<rgba 1 0 1 1>, paramNames [Panner, GradientVAdd, SplineThickenWidthMult, Param4]}.SplineThickenWidthMult
   )
  )
  n4 = Constant()
  n5 = ScalarParameter() {defaultValue=1, group=StalemateEmote, parameterName=StalemateEmote_AddedNoiseAmountMin}
  n6 = ScalarParameter() {defaultValue=1, group=StalemateEmote, parameterName=StalemateEmote_AddedNoiseAmountMax}
  n7 = ParticleRandom()
  n8 = ScalarParameter() {defaultValue=0.55, group=StalemateEmote, parameterName=StalemateEmote_ScaleVMin}
  n9 = ScalarParameter() {defaultValue=0.55, group=StalemateEmote, parameterName=StalemateEmote_ScaleVMin}
  n10 = ParticleRandom()
  n11 = Multiply(A=TextureCoordinate(), B=AppendVector(A=Constant(), B=LinearInterpolate(A=@n8, B=@n9, Alpha=@n10)))
  n12 = AppendVector(A=ParticleRandom(), B=ParticleRandom())
  n13 = ScalarParameter() {defaultValue=0.01, group=StalemateEmote, parameterName=StalemateEmote_PanU}
  n14 = ScalarParameter() {defaultValue=0.1, group=StalemateEmote, parameterName=StalemateEmote_PanV}
  n15 = TextureSample(UVs=Panner(Coordinate=Add(A=@n11, B=@n12), Speed=AppendVector(A=@n13, B=@n14))) {samplerType=SAMPLERTYPE_LinearColor, texture=&HighContrastNoise}
  n16 = MaterialFunctionCall(X=@n4, Y=@n4, Z=Multiply(A=LinearInterpolate(A=@n5, B=@n6, Alpha=@n7), B=LinearInterpolate(Alpha=@n15.R) {constA=-1}) {constA=1}) {materialFunction=&MakeFloat3}
  LinearInterpolate(
   A=Constant()
   B=LinearInterpolate(A=@n3, B=Add(A=@n3, B=@n16.Result), Alpha=ScalarParameter() {group=StalemateEmote, parameterName=StalemateEmote_UseAddedWPONoise})
   Alpha=ScalarParameter() {parameterName=UseSplineThicken}
  )
  n17 = ParticleColor()
  n18 = Multiply(A=TextureCoordinate(), B=ComponentMask(None=VectorParameter() {parameterName=UV Scale, defaultValue=<rgba 1 1 0 0>}.RGB) {g=true, r=true})
  n19 = DynamicParameter() {parameterIndex=1, defaultValue=<rgba 1 0 1 1>, paramNames [OffsetU,OffsetV,OffsetU2,OffsetV2]}
  n20 = Add(
   A=LinearInterpolate(
    A=@n18
    B=MaterialFunctionCall(XY=@n18) {materialFunction=&Swizzle}.YX
    Alpha=ScalarParameter() {parameterName=SwizzleUVs}
   )
   B=MaterialFunctionCall(X=@n19.OffsetU, Y=@n19.OffsetU) {materialFunction=&MakeFloat2}.Result
  )
  n21 = VectorParameter() {parameterName=PannerSpeed, defaultValue=<rgba 0 0 0 1>}
  n22 = MaterialFunctionCall(R=@n21.R, G=@n21.R) {materialFunction=&AppendMany}
  n23 = StaticComponentMaskParameter(None=TextureSampleParameter2D(UVs=Panner(Coordinate=@n20, Speed=@n22.RG)) {parameterName=Default Tex, texture=/Game/Effects/Fort_Effects/Emotes/Energize_Max/Burst/T_Emote_Energize_airflow}.RGB) {defaultR=true, parameterName=Component mask}
  n24 = DynamicParameter() {defaultValue=<rgba 1 0 1 1>, paramNames [Panner, GradientVAdd, SplineThickenWidthMult, Param4]}
  n25 = Add(A=TextureCoordinate(), B=Constant2Vector())
  n26 = DynamicParameter() {parameterIndex=1, defaultValue=<rgba 1 0 1 1>, paramNames [OffsetU,OffsetV,OffsetU2,OffsetV2]}
  n27 = LinearInterpolate(
   A=@n25
   B=Add(
    A=@n25
    B=MaterialFunctionCall(X=@n26.OffsetU2, Y=@n26.OffsetU2) {materialFunction=&MakeFloat2}.Result
   )
   Alpha=ScalarParameter() {parameterName=UseDynparamForHeightLerp}
  )
  n28 = VectorParameter() {parameterName=HeigthLerpUVMult, defaultValue=<rgba 0 0 0 1>}
  n29 = MaterialFunctionCall(X=@n28.R, Y=@n28.R) {materialFunction=&MakeFloat2}
  n30 = Multiply(A=TextureSample(UVs=Multiply(A=@n27, B=@n29.Result)) {samplerType=SAMPLERTYPE_LinearGrayscale, texture=/Game/Global/Textures/Default/Noise/T_SwirlNoise_N_height}.R)
  n31 = ScalarParameter() {defaultValue=0.5, parameterName="Contrast (S)"}
  n32 = ScalarParameter() {parameterName=Edge fade}
  n33 = LinearInterpolate(A=ScalarParameter() {parameterName=InvertEdgeFade}, B=@n32, Alpha=ConstantBiasScale(None=TwoSidedSign()))
  n34 = Multiply(
   A=Multiply(
    A=Multiply(A=@n17.A, B=@n23)
    B=MaterialFunctionCall(Transition Phase=@n24.Panner, Height Texture=@n30, Contrast=@n31) {materialFunction=&HeightLerp}.Alpha
   )
   B=LinearInterpolate(
    A=MaterialFunctionCall("FadeSensitivity (optional)"=@n32) {materialFunction=&FadeBasedOnViewAngle}.Result
    B=MaterialFunctionCall("FadeSensitivity (optional)"=@n33) {materialFunction=&FadeBasedOnViewAngle}.Result
    Alpha=ScalarParameter() {parameterName=UseTwoSidedFadeControls}
   )
  )
  n35 = DepthFade(Opacity=@n34, FadeDistance=ScalarParameter() {defaultValue=30, parameterName=DepthFade}) {fadeDistanceDefault=30}
  n36 = ShadingPathSwitch(Default=@n35, Deferred=@n35, Forward=@n34, Mobile=@n34)
  n37 = Multiply(A=ComponentMask(None=TextureCoordinate()) {r=true}) {constB=0.5}
  n38 = MaterialFunctionCall(Float2=TextureCoordinate()) {materialFunction=&BreakOutFloat2Components}
  n39 = DynamicParameter() {defaultValue=<rgba 1 0 1 1>, paramNames [Panner, GradientVAdd, SplineThickenWidthMult, Param4]}
  n40 = TextureSample(UVs=MaterialFunctionCall(X=@n38.R, Y=Add(A=@n38.G, B=@n39.GradientVAdd)) {materialFunction=&MakeFloat2}.Result) {samplerSource=SSM_Clamp_WorldGroupSettings, texture=/Game/Effects/Fort_Effects/Emotes/Energize_Max/Burst/T_FX_Emote_Stalemate_Gradient}
  n41 = Multiply(A=Power(Base=Sine(None=@n37)), B=@n40.RGB)
  n42 = VertexColor()
  LinearInterpolate(A=@n36, B=Multiply(A=Multiply(A=@n41, B=@n42.R), B=@n36), Alpha=ScalarParameter() {parameterName=Mult_VertColor_and_Edges})
  n43 = TextureCoordinate()
  n44 = LinearInterpolate(Alpha=SmoothStep(Min=Constant() {r=0.49}, Value=ComponentMask(None=@n43) {g=true})) {constA=0.03}
  n45 = Multiply(
   A=TextureCoordinate()
   B=ComponentMask(None=VectorParameter() {parameterName=Color noise UV Sacle, defaultValue=<rgba 0.3 0.5 0 0>}.RGB) {g=true, r=true}
  )
  n46 = ComponentMask(None=VectorParameter() {parameterName=Color noise UV offset, defaultValue=<rgba 0.5 0.5 0 1>}.RGB) {g=true, r=true}
  n47 = TextureSample(UVs=Panner(Coordinate=Add(A=@n45, B=@n46))) {samplerType=SAMPLERTYPE_Grayscale, texture=&LowResBlurredNoise}
  n48 = ScalarParameter() {defaultValue=1.3, parameterName=Color noise Power}
  n49 = LinearInterpolate(
   A=Saturate(None=Multiply(A=@n44) {constB=5})
   B=Multiply(B=Power(Base=@n47.R, Exp=@n48) {constExponent=1.3}) {constA=1}
   Alpha=ScalarParameter() {parameterName=Use detail color mask}
  )
  MaterialFunctionCall(Emissive=LinearInterpolate(A=Multiply(A=@n49, B=@n17.RGB), B=@n17.RGB, Alpha=ScalarParameter() {parameterName=UseParitcleColorEmissiveOnly})) {materialFunction=&MF_FX_EmissiveAdjustments}
  TextureCoordinate()
  TextureCoordinate() {uTiling=0.3, vTiling=0.5}
 ]
}
