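# Translucent, unlit, two-sided Niagara mesh-particle material that draws a thin airflow streak from a spiral line mask.
# Color, intensity and particle color feed an emissive-adjustment function, while particle alpha multiplies the mask's red channel.
# Dynamic parameters drive the UVs per particle: Random offsets U, Pan replaces time to scroll U, and Tile scales U.
# A scalar blends between plain and center-based UV scaling.
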Material /Game/Effects/Niagara/MasterMaterials/Airflow/M_AirflowSharp_Thin
graph {
 properties {
  bUsedWithNiagaraMeshParticles=true
  blendMode=BLEND_Translucent
  pixelDepthOffsetMode=PDOM_Legacy
  shadingModel=MSM_Unlit
  twoSided=true
  usedShadingModels=Unlit
 }
 nodes [
  n1 = VectorParameter() {parameterName=Color, defaultValue=<rgba 0 0.3333 1 1>}
  n2 = ScalarParameter() {defaultValue=3, parameterName=Intensity}
  n3 = ParticleColor()
  MaterialFunctionCall(Emissive=Multiply(A=Multiply(A=@n1.RGB, B=@n2), B=@n3.RGB)) {materialFunction=&MF_FX_EmissiveAdjustments}
  n4 = DynamicParameter() {defaultValue=<rgba 0 0 1 0>, paramNames [Random,Pan,Tile,Param4]}
  n5 = AppendVector(A=@n4.Random, B=Constant())
  n6 = TextureCoordinate()
  n7 = AppendVector(A=@n4.Tile, B=ScalarParameter() {defaultValue=1, parameterName=U_Scale})
  n8 = Panner(
   Coordinate=LinearInterpolate(
    A=Multiply(A=@n6, B=@n7)
    B=MaterialFunctionCall(UVs=@n6, Texture Scale=@n7) {materialFunction=&ScaleUVsByCenter}.UVs
    Alpha=ScalarParameter() {parameterName=Use_ScaleUVsByCenter}
   )
   Time=@n4.Pan
  ) {speedX=1}
  Multiply(
   A=@n3.A
   B=TextureSample(UVs=Add(A=@n5, B=@n8)) {samplerType=SAMPLERTYPE_Masks, texture=/Game/Effects/Niagara/MasterMaterials/Airflow/Texture/T_Emote_Spiral_Line}.R
  )
 ]
}
