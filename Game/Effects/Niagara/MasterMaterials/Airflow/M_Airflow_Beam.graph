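# Translucent unlit master material for Niagara airflow beam effects on sprites, ribbons and meshes.
# Emissive layers panning directional, cell and smoke noise, some offset per particle, under a V gradient, adds a plasma texture and tints by particle color.
# Opacity multiplies static and scrolling trail masks by particle alpha, an optional U-edge falloff and view-angle fade.
# A shading path switch applies depth fade only on default and deferred paths.
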Material /Game/Effects/Niagara/MasterMaterials/Airflow/M_Airflow_Beam
graph {
 properties {
  bAutomaticallySetUsageInEditor=false
  bUsedWithNiagaraMeshParticles=true
  bUsedWithNiagaraRibbons=true
  bUsedWithNiagaraSprites=true
  blendMode=BLEND_Translucent
  shadingModel=MSM_Unlit
  usedShadingModels=Unlit
 }
 nodes [
  n1 = TextureCoordinate() {uTiling=0.6, vTiling=0.6}
  n2 = AppendVector(A=Constant(), B=ParticleRandom())
  n3 = TextureSample(UVs=Panner(Coordinate=Add(A=@n1, B=@n2)) {speedX=-0.1, speedY=0.1}) {samplerType=SAMPLERTYPE_LinearGrayscale, texture=/Game/Effects/Fort_Effects/Emotes/UltraLight/Texture/T_DirectionalNoise_Greyscale}
  n4 = TextureCoordinate() {uTiling=0.7, vTiling=0.5}
  n5 = AppendVector(A=Constant(), B=ParticleRandom())
  n6 = TextureSample(UVs=Panner(Coordinate=Add(A=@n4, B=@n5))) {samplerType=SAMPLERTYPE_Grayscale, texture=/Game/Weapons/FORT_Melee/Pickaxe_Cube_Ninja_Male/FX/T_CellNoise}
  n7 = TextureCoordinate() {uTiling=0, vTiling=0}
  n8 = LinearInterpolate(Alpha=TextureSample(UVs=Panner(Coordinate=@n7)) {texture=/Game/Global/Textures/Default/Noise/Noise}.R) {constB=4}
  n9 = Multiply(A=Power(Base=Multiply(A=@n6.R)), B=Multiply(A=Saturate(None=@n8)) {constB=5})
  n10 = TextureSample(UVs=Panner(Coordinate=TextureCoordinate() {vTiling=0.6})) {samplerType=SAMPLERTYPE_LinearColor, texture=/Game/Effects/Fort_Effects/Gameplay/SmokeDustExplosions/Textures/T_RibbonSmokeMasked_03}
  n11 = Multiply(A=@n10.R)
  n12 = ComponentMask(None=Add(A=TextureCoordinate(), B=Constant2Vector() {g=-0.23})) {g=true}
  n13 = Multiply(A=Add(A=Multiply(A=@n3.B), B=Add(A=@n9, B=@n11)), B=Power(Base=OneMinus(None=@n12)))
  n14 = TextureSampleParameter2D(UVs=Panner(Coordinate=TextureCoordinate() {vTiling=0.5}) {speedX=0.3, speedY=0.15}) {
   parameterName=LerpTexture
   samplerType=SAMPLERTYPE_LinearGrayscale
   texture=/Game/Effects/Fort_Effects/Emotes/UltraLight/Texture/T_Plasma_A_B_Greyscale
  }
  n15 = Multiply(A=@n14.B) {constB=0.6}
  n16 = ParticleColor()
  MaterialFunctionCall(Emissive=Multiply(A=Add(A=@n13, B=@n15), B=@n16.RGB)) {materialFunction=&MF_FX_EmissiveAdjustments}
  n17 = Multiply(A=TextureCoordinate(), B=AppendVector(A=ScalarParameter() {defaultValue=1, parameterName=TrailMaskTexture_UTIling}, B=Constant() {r=1}))
  n18 = Multiply(
   A=TextureSampleParameter2D() {parameterName=MaskTexture, texture=/Game/Effects/Niagara/MasterMaterials/Airflow/Texture/T_Airflow_Mask001}.R
   B=TextureSampleParameter2D(UVs=Panner(Coordinate=@n17) {speedX=0.2}) {parameterName=TrailMaskTexture, texture=/Game/Effects/Niagara/MasterMaterials/Airflow/Texture/T_Airflow_MaskTrail}.R
  )
  n19 = ComponentMask(None=TextureCoordinate()) {r=true}
  n20 = Multiply(A=Multiply(A=@n19, B=OneMinus(None=@n19))) {constB=5}
  n21 = Multiply(A=@n18, B=Saturate(None=Power(Base=@n20)))
  n22 = ScalarParameter() {parameterName=Use_RightLinearMask}
  n23 = Multiply(
   A=Multiply(A=@n16.A, B=LinearInterpolate(A=@n18, B=@n21, Alpha=@n22))
   B=MaterialFunctionCall("FadeSensitivity (optional)"=ScalarParameter() {defaultValue=1.5, parameterName=MeshFade}) {materialFunction=&FadeBasedOnViewAngle}.Result
  )
  n24 = DepthFade(Opacity=@n23, FadeDistance=ScalarParameter() {defaultValue=5, parameterName=DepthFade})
  ShadingPathSwitch(Default=@n24, Deferred=@n24, Forward=@n23, Mobile=@n23)
 ]
}
