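# Translucent, unlit, two-sided Niagara master material for soft, scrolling airflow streaks on mesh particles and ribbons.
# Panning trail textures, with named reroutes for UV scale, offset and noise disturbance, are multiplied by blurred noise, edge fades and particle alpha.
# Scalar parameters blend toward a remapped "round" mask or a procedural glow mask.
# Particle color feeds an emissive adjustment function; dynamic parameters drive SplineThicken width and optional U offset and scale.
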Material /Game/Effects/Niagara/MasterMaterials/Airflow/M_SoftAirFlow_Square
graph {
 properties {
  bAutomaticallySetUsageInEditor=false
  bUsedWithNiagaraMeshParticles=true
  bUsedWithNiagaraRibbons=true
  blendMode=BLEND_Translucent
  pixelDepthOffsetMode=PDOM_Legacy
  shadingModel=MSM_Unlit
  twoSided=true
  usedShadingModels=Unlit
 }
 nodes [
  n1 = ParticleColor()
  n2 = Multiply(A=TextureCoordinate() {uTiling=0}, B=NamedRerouteUsage())
  n3 = NamedRerouteUsage()
  n4 = Panner(Coordinate=Add(A=Add(A=@n2, B=@n3), B=NamedRerouteUsage())) {speedX=-0.8}
  n5 = TextureSample(UVs=@n4) {
   samplerSource=SSM_Wrap_WorldGroupSettings
   texture=/Game/Weapons/FORT_Melee/Pickaxe_Future_Samurai_Summer_Male/FX/T_Pickaxe_Future_Samurai_Summer_Male_Trail_01
  }
  n6 = Min(
   A=@n5.R
   B=TextureSample(UVs=@n4) {samplerSource=SSM_Wrap_WorldGroupSettings, texture=/Game/Weapons/FORT_Melee/Pickaxe_Tower_Sentinel_Female/FX/T_TowerSentinel_trail}.R
  )
  n7 = Multiply(A=TextureCoordinate() {uTiling=0.3}, B=NamedRerouteUsage())
  n8 = Multiply(
   A=Multiply(A=LinearInterpolate(A=@n6, B=@n5.R) {constAlpha=0.2})
   B=TextureSample(UVs=Panner(Coordinate=@n7) {speedX=-0.8}) {samplerType=SAMPLERTYPE_Grayscale, texture=&LowResBlurredNoise}.R
  )
  n9 = Sine(None=ComponentMask(None=TextureCoordinate()) {r=true}) {period=2}
  n10 = TextureCoordinate()
  n11 = Multiply(A=OneMinus(None=ComponentMask(None=@n10) {r=true})) {constB=2}
  n12 = Multiply(A=Multiply(A=@n1.A) {constB=3}, B=Multiply(A=Multiply(A=@n8, B=@n9), B=Saturate(None=@n11)))
  n13 = TextureCoordinate()
  n14 = ComponentMask(None=@n13) {r=true}
  n15 = ComponentMask(None=@n13) {r=true}
  n16 = Multiply(A=Saturate(None=Multiply(A=@n14) {constB=2}), B=Power(Base=OneMinus(None=@n15)))
  n17 = TextureCoordinate()
  n18 = Power(Base=Sine(None=ComponentMask(None=@n17) {g=true}) {period=2})
  LinearInterpolate(
   A=LinearInterpolate(A=@n12, B=LinearInterpolate(Alpha=@n12) {constA=-0.3}, Alpha=ScalarParameter() {parameterName=Use Round mask})
   B=Multiply(A=@n1.A, B=Multiply(A=@n16, B=@n18))
   Alpha=ScalarParameter() {parameterName=UseGlowMask}
  )
  MaterialFunctionCall(Emissive=@n1.RGB) {materialFunction=&MF_FX_EmissiveAdjustments}
  n19 = DynamicParameter() {defaultValue=<rgba 0.25 50 1 1>, paramNames [Panner,SplineWidth,Uoffset,UScale]}
  MaterialFunctionCall(
   WidthBase=@n19.SplineWidth
   WidthTip=@n19.SplineWidth
   Expand U or V UV Channel=StaticBoolParameter() {parameterName="Expand U or V UV Channel (SB)"}
  ) {materialFunction=&SplineThicken}
  n20 = DynamicParameter() {defaultValue=<rgba 0.25 50 1 1>, paramNames [Panner,SplineWidth,Uoffset,UScale]}
  n21 = Constant()
  NamedRerouteDeclaration(None=LinearInterpolate(B=AppendVector(A=@n20.Uoffset, B=@n21), Alpha=ScalarParameter() {parameterName=Use custom offset})) {name=UOffset, nodeColor=<rgba 1 0 0.6328 1>}
  n22 = DynamicParameter() {defaultValue=<rgba 0.25 50 1 1>, paramNames [Panner,SplineWidth,Uoffset,UScale]}
  n23 = Constant() {r=1}
  NamedRerouteDeclaration(None=LinearInterpolate(B=AppendVector(A=@n22.UScale, B=@n23), Alpha=ScalarParameter() {parameterName=Use CustomUScale}) {constA=1}) {name=UScale, nodeColor=<rgba 0.1484 1 0 1>}
  n24 = TextureSample(UVs=Panner(Coordinate=TextureCoordinate() {uTiling=0.2}) {speedX=-0.3}) {samplerType=SAMPLERTYPE_Grayscale, texture=&LowResBlurredNoise}
  NamedRerouteDeclaration(None=Multiply(A=Subtract(A=@n24.R) {constB=0}) {constB=0.1}) {name=Disturb, nodeColor=<rgba 1 0.4768 0 1>}
 ]
}
