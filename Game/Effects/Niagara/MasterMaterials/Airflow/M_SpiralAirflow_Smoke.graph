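# Translucent, unlit, two-sided Niagara master material for scrolling spiral airflow smoke on mesh particles and ribbons.
# A panned inky smoke texture tinted by particle color feeds emissive adjustments, while SplineThicken takes its width from a dynamic parameter.
# Opacity multiplies soft trail, fluid and blurred noise textures by particle alpha, with sine and linear fades along U.
# Named reroutes share U offset and scale, blended toward per-particle dynamic parameters by scalar switches.
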Material /Game/Effects/Niagara/MasterMaterials/Airflow/M_SpiralAirflow_Smoke
graph {
 properties {
  bAutomaticallySetUsageInEditor=false
  bUsedWithNiagaraMeshParticles=true
  bUsedWithNiagaraRibbons=true
  blendMode=BLEND_Translucent
  pixelDepthOffsetMode=PDOM_Legacy
  shadingModel=MSM_Unlit
  twoSided=true
  usedShadingModels=Unlit
 }
 nodes [
  Constant2Vector() {r=1}
  n1 = Multiply(A=TextureCoordinate(), B=NamedRerouteUsage())
  n2 = NamedRerouteUsage()
  n3 = TextureSample(UVs=Panner(Coordinate=Add(A=@n1, B=@n2)) {speedX=-0.8}) {texture=&T_Inky_Smoke_Tile}
  n4 = ParticleColor()
  MaterialFunctionCall(Emissive=Multiply(A=LinearInterpolate(Alpha=@n3.R), B=@n4.RGB)) {materialFunction=&MF_FX_EmissiveAdjustments}
  n5 = DynamicParameter() {defaultValue=<rgba 0.25 50 1 1>, paramNames [Panner,SplineWidth,Uoffset,UScale]}
  MaterialFunctionCall(
   WidthBase=@n5.SplineWidth
   WidthTip=@n5.SplineWidth
   Expand U or V UV Channel=StaticBoolParameter() {parameterName="Expand U or V UV Channel (SB)"}
  ) {materialFunction=&SplineThicken}
  n6 = Multiply(A=TextureCoordinate(), B=NamedRerouteUsage())
  n7 = NamedRerouteUsage()
  n8 = TextureSample(UVs=Panner(Coordinate=Add(A=@n6, B=@n7)) {speedX=-0.8}) {
   samplerSource=SSM_Wrap_WorldGroupSettings
   samplerType=SAMPLERTYPE_Grayscale
   texture=&T_Smoke_Trail_Soft
  }
  n9 = TextureCoordinate() {vTiling=0.8}
  n10 = VectorParameter() {parameterName=NoiseUVScale, channelNames {a="", b="", g=VScale, r=UScale}, defaultValue=<rgba 1 1 1 0>}
  n11 = ComponentMask(None=@n10) {g=true, r=true}
  n12 = Panner(Coordinate=Add(A=Multiply(A=@n9, B=@n11), B=NamedRerouteUsage())) {speedX=-0.8}
  n13 = TextureCoordinate() {uTiling=1.5}
  n14 = NamedRerouteUsage()
  n15 = Multiply(
   A=Multiply(
    A=@n8.R
    B=TextureSample(UVs=@n12) {samplerType=SAMPLERTYPE_LinearGrayscale, texture=/Game/Weapons/FORT_Melee/Pickaxe_Cube_Ninja_Male/FX/T_FluidFlameTile}.R
   )
   B=TextureSample(UVs=Multiply(A=@n13, B=@n14)) {samplerType=SAMPLERTYPE_Grayscale, texture=&LowResBlurredNoise}.R
  )
  n16 = Sine(None=ComponentMask(None=TextureCoordinate()) {r=true}) {period=2}
  n17 = TextureCoordinate()
  n18 = Multiply(A=OneMinus(None=ComponentMask(None=@n17) {r=true})) {constB=2}
  Multiply(A=@n4.A, B=Multiply(A=Multiply(A=@n15, B=@n16), B=Saturate(None=@n18)))
  n19 = DynamicParameter() {defaultValue=<rgba 0.25 50 1 1>, paramNames [Panner,SplineWidth,Uoffset,UScale]}
  n20 = Constant()
  NamedRerouteDeclaration(None=LinearInterpolate(B=AppendVector(A=@n19.Uoffset, B=@n20), Alpha=ScalarParameter() {parameterName=Use custom offset})) {name=UOffset, nodeColor=<rgba 1 0 0.6328 1>}
  n21 = DynamicParameter() {defaultValue=<rgba 0.25 50 1 1>, paramNames [Panner,SplineWidth,Uoffset,UScale]}
  n22 = Constant() {r=1}
  NamedRerouteDeclaration(None=LinearInterpolate(B=AppendVector(A=@n21.UScale, B=@n22), Alpha=ScalarParameter() {parameterName=Use CustomUScale}) {constA=1}) {name=UScale, nodeColor=<rgba 0.1484 1 0 1>}
  NamedRerouteUsage()
 ]
}
