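# Translucent, unlit, two-sided master material for spiral airflow trails on Niagara mesh particles.
# A dynamic Panner parameter scrolls the trail mask, which is multiplied by noise, UV edge gradients and particle alpha.
# View-angle, near-camera and depth fades soften opacity, while particle color feeds an emissive adjustment function.
# An optional SplineThicken offset sets mesh width from a dynamic SplineWidth parameter.
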Material /Game/Effects/Niagara/MasterMaterials/Airflow/M_Spiral_Airflow
graph {
 properties {
  bAutomaticallySetUsageInEditor=false
  bUsedWithNiagaraMeshParticles=true
  blendMode=BLEND_Translucent
  pixelDepthOffsetMode=PDOM_Legacy
  shadingModel=MSM_Unlit
  twoSided=true
  usedShadingModels=Unlit
 }
 nodes [
  Add(A=TextureCoordinate() {uTiling=1.5, vTiling=0.7}, B=Constant2Vector() {g=0.2, r=1})
  Constant() {r=50}
  n1 = VectorParameter() {parameterName=Trail UV Control, channelNames {a=Offset V, b=Offset U, g=Tiling V, r=Tiling U}, defaultValue=<rgba 1 1 1 0.2>}
  n2 = Multiply(A=TextureCoordinate() {uTiling=1.5, vTiling=0.7}, B=AppendVector(A=@n1.Tiling U, B=@n1.Tiling U))
  n3 = AppendVector(A=@n1.Offset U, B=@n1.Offset U)
  n4 = DynamicParameter() {defaultValue=<rgba 0 50 1 1>, paramNames [Panner,SplineWidth,Param3,Param4]}
  n5 = TextureSampleParameter2D(UVs=Panner(Coordinate=Add(A=@n2, B=@n3), Time=LinearInterpolate(Alpha=@n4.Panner) {constB=8}) {speedX=-0.5}) {
   parameterName=Trail Texture
   samplerSource=SSM_Clamp_WorldGroupSettings
   samplerType=SAMPLERTYPE_LinearGrayscale
   texture=/Game/Gadgets/Assets/VinderTech_GliderChute/Glider_Ensemble_Maroon_Male/FX/Textures/T_RibbonMask_B
  }
  n6 = TextureCoordinate() {uTiling=1.5}
  n7 = Multiply(
   A=@n5.R
   B=LinearInterpolate(
    B=TextureSampleParameter2D(UVs=@n6) {parameterName=Noise Texture, samplerType=SAMPLERTYPE_Grayscale, texture=&LowResBlurredNoise}.R
    Alpha=ScalarParameter() {defaultValue=1, parameterName=Noise Impact}
   ) {constA=1}
  )
  n8 = Sine(None=ComponentMask(None=TextureCoordinate()) {r=true}) {period=2}
  n9 = ScalarParameter() {parameterName=HardEdgeGradientOpacityLerp}
  n10 = LinearInterpolate(A=@n8, B=Saturate(None=Multiply(A=@n8) {constB=6}), Alpha=@n9)
  n11 = ComponentMask(None=TextureCoordinate()) {r=true}
  n12 = Saturate(None=Multiply(A=OneMinus(None=@n11)) {constB=2})
  n13 = ParticleColor()
  n14 = Multiply(A=@n13.A, B=Multiply(A=Multiply(A=@n7, B=@n10), B=LinearInterpolate(A=@n12, Alpha=@n9)))
  n15 = MaterialFunctionCall("FadeSensitivity (optional)"=ScalarParameter() {parameterName=MeshFade}) {materialFunction=&FadeBasedOnViewAngle}
  n16 = ScalarParameter() {parameterName=Near Camera Fade}
  DepthFade(
   Opacity=Multiply(
    A=Multiply(A=@n14, B=@n15.Result)
    B=MaterialFunctionCall(Distance=@n16) {materialFunction=&Near_Camera_Fade}.Result
   )
   FadeDistance=ScalarParameter() {defaultValue=2, parameterName=DepthFade}
  )
  n17 = DynamicParameter() {defaultValue=<rgba 0 50 1 1>, paramNames [Panner,SplineWidth,Param3,Param4]}
  LinearInterpolate(
   B=MaterialFunctionCall(
    WidthBase=@n17.SplineWidth
    WidthTip=@n17.SplineWidth
    Expand U or V UV Channel=StaticBoolParameter() {parameterName="Expand U or V UV Channel (SB)"}
   ) {materialFunction=&SplineThicken}.WorldPositionOffset
   Alpha=ScalarParameter() {defaultValue=1, parameterName=Use SplineThicken}
  ) {constAlpha=1}
  MaterialFunctionCall(Emissive=@n13.RGB) {materialFunction=&MF_FX_EmissiveAdjustments}
  MaterialFunctionCall() {materialFunction=&TimeWithSpeedVariable}
 ]
}
